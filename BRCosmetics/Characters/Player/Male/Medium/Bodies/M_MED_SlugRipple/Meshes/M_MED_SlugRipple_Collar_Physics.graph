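# Physics asset for the collar of the SlugRipple male medium body cosmetic, defining bodies and joints for secondary collar motion.
# Eight two-capsule dyn_collar chains with tight swing and twist limits hang from base bodies joined to deform_collar, which links to spine_05.
# Larger capsules cover spine, neck, clavicles and upper arms, with angularly free joints.
# Angularly locked cross-links tie each side tip to its front and back diagonal neighbors, leaving center front and back chains unlinked.
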PhysicsAsset /BRCosmetics/Characters/Player/Male/Medium/Bodies/M_MED_SlugRipple/Meshes/M_MED_SlugRipple_Collar_Physics
bodies {
 base_collar_bk [{shapeType=Capsule, length=0.505, radius=0.505, shapeName=base_collar_bk_capsule, rotation=<rot 0 -90 0>}]
 base_collar_bk_l [{shapeType=Capsule, length=0.505, radius=0.505, shapeName=base_collar_bk_l_capsule, rotation=<rot 0 -90 0>}]
 base_collar_bk_r [{shapeType=Capsule, length=0.505, radius=0.505, shapeName=base_collar_bk_r_capsule, rotation=<rot 0 -90 0>}]
 base_collar_front [{shapeType=Capsule, length=0.505, radius=0.505, shapeName=base_collar_front_capsule, rotation=<rot 0 -90 0>}]
 base_collar_front_l [{shapeType=Capsule, length=0.505, radius=0.505, shapeName=base_collar_front_l_capsule, rotation=<rot 0 -90 0>}]
 base_collar_front_r [{shapeType=Capsule, length=0.505, radius=0.505, shapeName=base_collar_front_r_capsule, rotation=<rot 0 -90 0>}]
 base_collar_l [{shapeType=Capsule, length=0.505, radius=0.505, shapeName=base_collar_l_capsule, rotation=<rot 0 -90 0>}]
 base_collar_r [{shapeType=Capsule, length=0.505, radius=0.505, shapeName=base_collar_r_capsule, rotation=<rot 0 -90 0>}]
 clavicle_l [{shapeType=Capsule, length=9.5192, radius=9.3871, center=<9.4266 -1.1806 -2.4977>, rotation=<rot 69.3297 1.4476 176.9462>}]
 clavicle_r [{shapeType=Capsule, length=9.5192, radius=9.3871, center=<-9.4266 1.1806 2.4977>, rotation=<rot 69.3297 1.4476 -3.0538>}]
 deform_collar [{shapeType=Capsule, length=0.505, radius=0.505, shapeName=base_collar_bk_capsule, rotation=<rot 0 -90 0>}]
 dyn_collar_bk_01 [{shapeType=Capsule, length=4.5, radius=2.5, shapeName=dyn_collar_bk_01_capsule, center=<2 2 0>}]
 dyn_collar_bk_02 [{shapeType=Capsule, length=4.5, radius=2.5, shapeName=dyn_collar_bk_02_capsule, center=<2 2 0>}]
 dyn_collar_bk_l_01 [{shapeType=Capsule, length=4.5, radius=2.5, shapeName=dyn_collar_bk_l_01_capsule, center=<2 2 0>}]
 dyn_collar_bk_l_02 [{shapeType=Capsule, length=4.5, radius=2.5, shapeName=dyn_collar_bk_l_02_capsule, center=<2 2 0>}]
 dyn_collar_bk_r_01 [{shapeType=Capsule, length=4.5, radius=2.5, shapeName=dyn_collar_bk_r_01_capsule, center=<2 2 0>}]
 dyn_collar_bk_r_02 [{shapeType=Capsule, length=4.5, radius=2.5, shapeName=dyn_collar_bk_r_02_capsule, center=<2 2 0>}]
 dyn_collar_front_01 [{shapeType=Capsule, length=4.5, radius=2.5, shapeName=dyn_collar_front_01_capsule, center=<2 2 0>}]
 dyn_collar_front_02 [{shapeType=Capsule, length=4.5, radius=2.5, shapeName=dyn_collar_front_02_capsule, center=<2 2 0>}]
 dyn_collar_front_l_01 [{shapeType=Capsule, length=4.5, radius=2.5, shapeName=dyn_collar_front_l_01_capsule, center=<2 2 0>}]
 dyn_collar_front_l_02 [{shapeType=Capsule, length=4.5, radius=2.5, shapeName=dyn_collar_front_l_02_capsule, center=<2 2 0>}]
 dyn_collar_front_r_01 [{shapeType=Capsule, length=4.5, radius=2.5, shapeName=dyn_collar_front_r_01_capsule, center=<2 2 0>}]
 dyn_collar_front_r_02 [{shapeType=Capsule, length=4.5, radius=2.5, shapeName=dyn_collar_front_r_02_capsule, center=<2 2 0>}]
 dyn_collar_l_01 [{shapeType=Capsule, length=4.5, radius=2.5, shapeName=dyn_collar_l_01_capsule, center=<2 3 0>}]
 dyn_collar_l_02 [{shapeType=Capsule, length=4.5, radius=2.5, shapeName=dyn_collar_l_02_capsule, center=<2 3 0>}]
 dyn_collar_r_01 [{shapeType=Capsule, length=4.5, radius=2.5, shapeName=dyn_collar_r_01_capsule, center=<2 3 0>}]
 dyn_collar_r_02 [{shapeType=Capsule, length=4.5, radius=2.5, shapeName=dyn_collar_r_02_capsule, center=<2 3 0>}]
 neck_02 [{shapeType=Capsule, length=13.4375, radius=7.3175, center=<-0.2642 -0.3408 0>, rotation=<rot 90 -171.1193 -175.3978>}]
 spine_05 [{shapeType=Capsule, length=10.5226, radius=13.195, center=<-1.7729 -3.7366 0>}]
 upperarm_l [{shapeType=Capsule, length=20, radius=5, center=<8.0509 0 -1.3883>, rotation=<rot 90 0.0006 0>}]
 upperarm_r [{shapeType=Capsule, length=20, radius=5, center=<-8.0509 0 1.3883>, rotation=<rot -90 0 0>}]
}
constraints [
 {
  bone1Name=neck_02
  bone2Name=spine_05
  swing1LimitDegrees=45
  swing1Motion=Free
  swing2LimitDegrees=45
  swing2Motion=Free
  twistLimitDegrees=45
  twistMotion=Free
 }
 {
  bone1Name=clavicle_r
  bone2Name=spine_05
  swing1LimitDegrees=45
  swing1Motion=Free
  swing2LimitDegrees=45
  swing2Motion=Free
  twistLimitDegrees=45
  twistMotion=Free
 }
 {
  bone1Name=upperarm_r
  bone2Name=clavicle_r
  swing1LimitDegrees=45
  swing1Motion=Free
  swing2LimitDegrees=45
  swing2Motion=Free
  twistLimitDegrees=45
  twistMotion=Free
 }
 {
  bone1Name=upperarm_l
  bone2Name=clavicle_l
  swing1LimitDegrees=45
  swing1Motion=Free
  swing2LimitDegrees=45
  swing2Motion=Free
  twistLimitDegrees=45
  twistMotion=Free
 }
 {
  bone1Name=clavicle_l
  bone2Name=spine_05
  swing1LimitDegrees=45
  swing1Motion=Free
  swing2LimitDegrees=45
  swing2Motion=Free
  twistLimitDegrees=45
  twistMotion=Free
 }
 {bone1Name=dyn_collar_bk_01, bone2Name=base_collar_bk, swing1LimitDegrees=28, swing2LimitDegrees=14, twistLimitDegrees=5}
 {bone1Name=dyn_collar_bk_02, bone2Name=dyn_collar_bk_01, swing1LimitDegrees=28, swing2LimitDegrees=14, twistLimitDegrees=5}
 {bone1Name=dyn_collar_bk_l_01, bone2Name=base_collar_bk_l, swing1LimitDegrees=28, swing2LimitDegrees=14, twistLimitDegrees=5}
 {bone1Name=dyn_collar_bk_l_02, bone2Name=dyn_collar_bk_l_01, swing1LimitDegrees=28, swing2LimitDegrees=14, twistLimitDegrees=5}
 {bone1Name=dyn_collar_bk_r_01, bone2Name=base_collar_bk_r, swing1LimitDegrees=28, swing2LimitDegrees=14, twistLimitDegrees=5}
 {bone1Name=dyn_collar_bk_r_02, bone2Name=dyn_collar_bk_r_01, swing1LimitDegrees=28, swing2LimitDegrees=14, twistLimitDegrees=5}
 {bone1Name=dyn_collar_front_01, bone2Name=base_collar_front, swing1LimitDegrees=28, swing2LimitDegrees=14, twistLimitDegrees=5}
 {bone1Name=dyn_collar_front_02, bone2Name=dyn_collar_front_01, swing1LimitDegrees=28, swing2LimitDegrees=14, twistLimitDegrees=5}
 {bone1Name=dyn_collar_front_l_01, bone2Name=base_collar_front_l, swing1LimitDegrees=28, swing2LimitDegrees=14, twistLimitDegrees=5}
 {bone1Name=dyn_collar_front_l_02, bone2Name=dyn_collar_front_l_01, swing1LimitDegrees=28, swing2LimitDegrees=14, twistLimitDegrees=5}
 {bone1Name=dyn_collar_front_r_01, bone2Name=base_collar_front_r, swing1LimitDegrees=28, swing2LimitDegrees=14, twistLimitDegrees=5}
 {bone1Name=dyn_collar_front_r_02, bone2Name=dyn_collar_front_r_01, swing1LimitDegrees=28, swing2LimitDegrees=14, twistLimitDegrees=5}
 {bone1Name=dyn_collar_l_01, bone2Name=base_collar_l, swing1LimitDegrees=28, swing2LimitDegrees=14, twistLimitDegrees=5}
 {bone1Name=dyn_collar_l_02, bone2Name=dyn_collar_l_01, swing1LimitDegrees=28, swing2LimitDegrees=14, twistLimitDegrees=5}
 {bone1Name=dyn_collar_r_01, bone2Name=base_collar_r, swing1LimitDegrees=28, swing2LimitDegrees=14, twistLimitDegrees=5}
 {bone1Name=dyn_collar_r_02, bone2Name=dyn_collar_r_01, swing1LimitDegrees=28, swing2LimitDegrees=14, twistLimitDegrees=5}
 {
  bone1Name=deform_collar
  bone2Name=spine_05
  swing1LimitDegrees=45
  swing1Motion=Free
  swing2LimitDegrees=45
  swing2Motion=Free
  twistLimitDegrees=45
  twistMotion=Free
 }
 {
  bone1Name=base_collar_bk
  bone2Name=deform_collar
  swing1LimitDegrees=45
  swing1Motion=Free
  swing2LimitDegrees=45
  swing2Motion=Free
  twistLimitDegrees=45
  twistMotion=Free
 }
 {
  bone1Name=base_collar_bk_l
  bone2Name=deform_collar
  swing1LimitDegrees=45
  swing1Motion=Free
  swing2LimitDegrees=45
  swing2Motion=Free
  twistLimitDegrees=45
  twistMotion=Free
 }
 {
  bone1Name=base_collar_bk_r
  bone2Name=deform_collar
  swing1LimitDegrees=45
  swing1Motion=Free
  swing2LimitDegrees=45
  swing2Motion=Free
  twistLimitDegrees=45
  twistMotion=Free
 }
 {
  bone1Name=base_collar_front
  bone2Name=deform_collar
  swing1LimitDegrees=45
  swing1Motion=Free
  swing2LimitDegrees=45
  swing2Motion=Free
  twistLimitDegrees=45
  twistMotion=Free
 }
 {
  bone1Name=base_collar_front_l
  bone2Name=deform_collar
  swing1LimitDegrees=45
  swing1Motion=Free
  swing2LimitDegrees=45
  swing2Motion=Free
  twistLimitDegrees=45
  twistMotion=Free
 }
 {
  bone1Name=base_collar_front_r
  bone2Name=deform_collar
  swing1LimitDegrees=45
  swing1Motion=Free
  swing2LimitDegrees=45
  swing2Motion=Free
  twistLimitDegrees=45
  twistMotion=Free
 }
 {
  bone1Name=base_collar_l
  bone2Name=deform_collar
  swing1LimitDegrees=45
  swing1Motion=Free
  swing2LimitDegrees=45
  swing2Motion=Free
  twistLimitDegrees=45
  twistMotion=Free
 }
 {
  bone1Name=base_collar_r
  bone2Name=deform_collar
  swing1LimitDegrees=45
  swing1Motion=Free
  swing2LimitDegrees=45
  swing2Motion=Free
  twistLimitDegrees=45
  twistMotion=Free
 }
 {
  bone1Name=dyn_collar_front_l_02
  bone2Name=dyn_collar_l_02
  swing1LimitDegrees=45
  swing1Motion=Locked
  swing2LimitDegrees=45
  swing2Motion=Locked
  twistLimitDegrees=45
  twistMotion=Locked
 }
 {
  bone1Name=dyn_collar_bk_l_02
  bone2Name=dyn_collar_l_02
  swing1LimitDegrees=45
  swing1Motion=Locked
  swing2LimitDegrees=45
  swing2Motion=Locked
  twistLimitDegrees=45
  twistMotion=Locked
 }
 {
  bone1Name=dyn_collar_front_r_02
  bone2Name=dyn_collar_r_02
  swing1LimitDegrees=45
  swing1Motion=Locked
  swing2LimitDegrees=45
  swing2Motion=Locked
  twistLimitDegrees=45
  twistMotion=Locked
 }
 {
  bone1Name=dyn_collar_bk_r_02
  bone2Name=dyn_collar_r_02
  swing1LimitDegrees=45
  swing1Motion=Locked
  swing2LimitDegrees=45
  swing2Motion=Locked
  twistLimitDegrees=45
  twistMotion=Locked
 }
]
