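# Physics asset for the M_MED_SlugRipple male medium body costume, defining simulation bodies for dynamic costume bones and a chest softball.
# Five tiny base capsules (front, back-left, back-right, left, right) lock angularly to spine_03, each carrying a dyn capsule with tight swing and twist limits.
# The softball base chains off the front dyn body with free rotation and carries its own limited dyn body.
# Two unconstrained spine_02 capsules add torso volume; costume shapes retain hoodie names.
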PhysicsAsset /BRCosmetics/Characters/Player/Male/Medium/Bodies/M_MED_SlugRipple/Meshes/M_MED_SlugRipple_Costume_Physics
bodies {
 base_chest_softball [{shapeType=Capsule, length=0.505, radius=0.505, shapeName=dyn_hoodie_fr_master_capsule, rotation=<rot 0 -90 0>}]
 base_costume_bk_l [{shapeType=Capsule, length=0.505, radius=0.505, shapeName=dyn_hoodie_bk_master_capsule, rotation=<rot 0 -90 0>}]
 base_costume_bk_r [{shapeType=Capsule, length=0.505, radius=0.505, shapeName=dyn_hoodie_bk_master_capsule, rotation=<rot 0 -90 0>}]
 base_costume_fr [{shapeType=Capsule, length=0.505, radius=0.505, shapeName=dyn_hoodie_fr_master_capsule, rotation=<rot 0 -90 0>}]
 base_costume_l [{shapeType=Capsule, length=0.505, radius=0.505, shapeName=dyn_hoodie_l_master_capsule, rotation=<rot 0 -90 0>}]
 base_costume_r [{shapeType=Capsule, length=0.505, radius=0.505, shapeName=dyn_hoodie_r_master_capsule, rotation=<rot 0 -90 0>}]
 dyn_chest_softball [{shapeType=Capsule, length=5, radius=3, shapeName=dyn_hoodie_fr_capsule, center=<0 2 0>}]
 dyn_costume_bk_l_01 [{shapeType=Capsule, length=8, radius=3, shapeName=dyn_hoodie_bk_capsule, center=<4 0 0>}]
 dyn_costume_bk_r_01 [{shapeType=Capsule, length=8, radius=3, shapeName=dyn_hoodie_bk_capsule, center=<4 0 0>}]
 dyn_costume_fr_01 [{shapeType=Capsule, length=8, radius=3, shapeName=dyn_hoodie_fr_capsule, center=<4 0 0>}]
 dyn_costume_l_01 [{shapeType=Capsule, length=8, radius=3, shapeName=dyn_hoodie_l_capsule, center=<4 0 0>}]
 dyn_costume_r_01 [{shapeType=Capsule, length=8, radius=3, shapeName=dyn_hoodie_r_capsule, center=<4 0 0>}]
 spine_02 [
  {shapeType=Capsule, length=8.9107, radius=6.04, shapeName=spine_02_capsule, center=<6.0263 2.1585 -0.072>, rotation=<rot -1.3876 -90 0.6242>}
  {
   shapeType=Capsule
   length=11.5561
   radius=6.8146
   shapeName=spine_02_capsule
   center=<13.1957 2.1577 -0.0412>
   rotation=<rot -1.3876 -90 0.6242>
  }
 ]
 spine_03 [{shapeType=Capsule, length=0.505, radius=0.505, shapeName=spine_03_capsule, rotation=<rot 0 -90 0>}]
}
constraints [
 {bone1Name=dyn_costume_bk_l_01, bone2Name=base_costume_bk_l, swing1LimitDegrees=10, swing2LimitDegrees=5, twistLimitDegrees=5}
 {bone1Name=dyn_costume_bk_r_01, bone2Name=base_costume_bk_r, swing1LimitDegrees=10, swing2LimitDegrees=5, twistLimitDegrees=5}
 {bone1Name=dyn_costume_fr_01, bone2Name=base_costume_fr, swing1LimitDegrees=10, swing2LimitDegrees=5, twistLimitDegrees=5}
 {bone1Name=dyn_costume_l_01, bone2Name=base_costume_l, swing1LimitDegrees=10, swing2LimitDegrees=5, twistLimitDegrees=5}
 {bone1Name=dyn_costume_r_01, bone2Name=base_costume_r, swing1LimitDegrees=10, swing2LimitDegrees=5, twistLimitDegrees=5}
 {
  bone1Name=base_costume_bk_l
  bone2Name=spine_03
  swing1LimitDegrees=45
  swing1Motion=Locked
  swing2LimitDegrees=45
  swing2Motion=Locked
  twistLimitDegrees=45
  twistMotion=Locked
 }
 {
  bone1Name=base_costume_bk_r
  bone2Name=spine_03
  swing1LimitDegrees=45
  swing1Motion=Locked
  swing2LimitDegrees=45
  swing2Motion=Locked
  twistLimitDegrees=45
  twistMotion=Locked
 }
 {
  bone1Name=base_costume_fr
  bone2Name=spine_03
  swing1LimitDegrees=45
  swing1Motion=Locked
  swing2LimitDegrees=45
  swing2Motion=Locked
  twistLimitDegrees=45
  twistMotion=Locked
 }
 {
  bone1Name=base_costume_l
  bone2Name=spine_03
  swing1LimitDegrees=45
  swing1Motion=Locked
  swing2LimitDegrees=45
  swing2Motion=Locked
  twistLimitDegrees=45
  twistMotion=Locked
 }
 {
  bone1Name=base_costume_r
  bone2Name=spine_03
  swing1LimitDegrees=45
  swing1Motion=Locked
  swing2LimitDegrees=45
  swing2Motion=Locked
  twistLimitDegrees=45
  twistMotion=Locked
 }
 {
  bone1Name=base_chest_softball
  bone2Name=dyn_costume_fr_01
  swing1LimitDegrees=45
  swing1Motion=Free
  swing2LimitDegrees=45
  swing2Motion=Free
  twistLimitDegrees=45
  twistMotion=Free
 }
 {bone1Name=dyn_chest_softball, bone2Name=base_chest_softball, swing1LimitDegrees=10, swing2LimitDegrees=5, twistLimitDegrees=5}
]
