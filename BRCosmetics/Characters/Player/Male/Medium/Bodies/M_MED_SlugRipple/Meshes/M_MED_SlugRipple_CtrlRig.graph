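# Post-animation corrective Control Rig for the M_MED_SlugRipple player body, fixing deformation of sleeves, collar, pants and joints.
# A shared DeformRig function runs first, then sequenced passes constrain sleeve and calf bones.
# Spherical pose readers and remapped bone rotation deltas weight additional bone offsets.
# Pistol, Melee, DBNO, Emote and NotRBAN flags gate sleeve and collar fixes, while ArmsHidden and LegsHidden shrink lower sleeves and pants.
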ControlRigBlueprint /BRCosmetics/Characters/Player/Male/Medium/Bodies/M_MED_SlugRipple/Meshes/M_MED_SlugRipple_CtrlRig
graphs [
 RigVMFunctionLibrary {links [], nodes []}
 RigVMFunction_Sequence_10_ContainedGraph {nodes [
   n1 = RigVMFunctionEntryNode()
   n2 = &RigVMFunction_Sequence(ExecuteContext=@n1.ExecuteContext)
   n3 = &RigVMFunction_Sequence(ExecuteContext=@n2.B)
   RigVMFunctionReturnNode(B=@n3.A, C=@n3.B, A=@n2.A)
  ]}
 RigVMFunction_Sequence_11_ContainedGraph {nodes [
   n1 = RigVMFunctionEntryNode()
   n2 = &RigVMFunction_Sequence(ExecuteContext=@n1.ExecuteContext)
   n3 = &RigVMFunction_Sequence(ExecuteContext=@n2.B)
   RigVMFunctionReturnNode(B=@n3.A, C=@n3.B, A=@n2.A)
  ]}
 RigVMFunction_Sequence_1_ContainedGraph {nodes [
   n1 = RigVMFunctionEntryNode()
   RigVMFunction_Sequence = &RigVMFunction_Sequence(ExecuteContext=@n1.ExecuteContext)
   n2 = &RigVMFunction_Sequence(ExecuteContext=@RigVMFunction_Sequence.B)
   RigVMFunctionReturnNode(A=@RigVMFunction_Sequence.A, B=@n2.A, C=@n2.B)
  ]}
 RigVMFunction_Sequence_1_ContainedGraph {nodes [
   n1 = RigVMFunctionEntryNode()
   RigVMFunction_Sequence = &RigVMFunction_Sequence(ExecuteContext=@n1.ExecuteContext)
   n2 = &RigVMFunction_Sequence(ExecuteContext=@RigVMFunction_Sequence.B)
   RigVMFunctionReturnNode(A=@RigVMFunction_Sequence.A, B=@n2.A, C=@n2.B)
  ]}
 RigVMFunction_Sequence_2_ContainedGraph {nodes [
   n1 = RigVMFunctionEntryNode()
   n2 = &RigVMFunction_Sequence(ExecuteContext=@n1.ExecuteContext)
   n3 = &RigVMFunction_Sequence(ExecuteContext=@n2.B)
   n4 = &RigVMFunction_Sequence(ExecuteContext=@n3.B)
   n5 = &RigVMFunction_Sequence(ExecuteContext=@n4.B)
   n6 = &RigVMFunction_Sequence(ExecuteContext=@n5.B)
   n7 = &RigVMFunction_Sequence(ExecuteContext=@n6.B)
   RigVMFunctionReturnNode(A=@n2.A, B=@n3.A, C=@n4.A, D=@n5.A, E=@n6.A, F=@n7.A, G=@n7.B)
  ]}
 RigVMModel {nodes [
   RigUnit_BeginExecution = &RigUnit_BeginExecution()
   FN_DeformRig_FNC_1 = RigVMFunctionReferenceNode(
    DeformRig_ Curve="(deformRig_l_upperarm_up_alpha_72_#1=False,deformRig_r_upperarm_up_alpha_73_#2=False,deformRig_l_upperarm_fr_alpha_74_#3=False,deformRig_r_upperarm_fr_alpha_75_CB51B8 ...(truncated)"
    DeformRig_Bool="(upperarm_1_#4=True,elbow_3_#5=True,wrist_7_#6=false,groin_8_#7=True,glute_17_#8=false,knee_9_40C408324CA2 ...(truncated)"
    DeformRig_Debug=(upperarm_debug_18_#4=False,wrist_debug_20_#6=False,lat_debug_26_#9=False,groinGlute_debug_27_#7=False)
    DeformRig_Value="(upperarm_up_18_#10=(Rotation=(X=0,Y=0,Z=0,W=1),Translation=(X=2,Y=0,Z=4),Scale3D=(X=0.7,Y=1,Z=1)),upperarm_fwd_fr_19_#11=(Rotation=(X=0,Y=0,Z=0,W=1),Translation=(X=0,Y=5,Z=0),Scale3D=(X=1,Y=1,Z=1)) ...(truncated)"
    ExecuteContext=@RigUnit_BeginExecution.ExecutePin
   )
   > RigVMFunction_Sequence_1 = RigVMAggregateNode()
   n1 = &RigUnit_SphericalPoseReader(
    ActiveRegionScaleFactors=(PositiveWidth=2.5,NegativeWidth=3,PositiveHeight=3,NegativeHeight=3)
    ActiveRegionSize=0.07
    Debug=(bDrawDebug=false,DebugScale=25,DebugSegments=20,DebugThickness=0.25)
    DriverAxis=(X=-1,Y=0,Z=0)
    DriverItem=(Type=Bone,Name="upperarm_r")
    FalloffRegionScaleFactors=(PositiveWidth=1,NegativeWidth=1,PositiveHeight=1,NegativeHeight=1)
    FalloffSize=0.5
    FlipHeightScaling=true
    OptionalParentItem=(Type=Bone)
    RotationOffset=(X=180,Y=-35,Z=140)
    ExecutePin=@RigVMFunction_Sequence_1.A
   )
   > &RigUnit_ModifyTransforms(
    ItemToModify="((Item=(Type=Bone,Name=\"deform_upperarm_fr_r\"),Transform=(Rotation=(X=0,Y=0,Z=0.3827,W=0.9239),Translation=(X=4,Y=0,Z=5),Scale3D=(X=1.2,Y=1.2,Z=1.2))),(Item=(Type=Bone,Name=\"upperarm_twist_01_r\"),Transform=(Rotation=(X=0.0436,Y=0,Z=0,W=0.999),Translation=( ...(truncated)"
    WeightMinimum=0
    Weight=@n1.OutputParam
   )
   > RigVMFunction_Sequence_9 = RigVMAggregateNode()
   n2 = &RigVMFunction_Sequence(ExecuteContext=@RigVMFunction_Sequence_9.A)
   Get Delta Ref_12 = RigVMFunctionReferenceNode(Item=(Type=Bone,Name="hand_r"), ExecuteContext=@n2.A)
   > &RigUnit_ModifyTransforms(
    ItemToModify=((Item=(Type=Bone,Name="base_handSleeveBk_r"),Transform=(Rotation=(X=0,Y=-0.1736,Z=0,W=0.9848),Translation=(X=0,Y=0,Z=1),Scale3D=(X=1,Y=1,Z=1))))
    WeightMinimum=0
    Weight=&RigVMFunction_MathDoubleRemap(SourceMaximum=58, SourceMinimum=0, TargetMaximum=1, TargetMinimum=0, Value=@"Get Delta Ref_12".Result.Rotation.Pitch).Result
   )
   Get Delta Ref_13 = RigVMFunctionReferenceNode(Item=(Type=Bone,Name="hand_l"), ExecuteContext=@n2.B)
   > &RigUnit_ModifyTransforms(
    ItemToModify=((Item=(Type=Bone,Name="dyn_handSleeveBk_l_01"),Transform=(Rotation=(X=0,Y=-0.1736,Z=0,W=0.9848),Translation=(X=0,Y=0,Z=-1),Scale3D=(X=1,Y=1,Z=1))))
    WeightMinimum=0
    Weight=&RigVMFunction_MathDoubleRemap(SourceMaximum=58, SourceMinimum=0, TargetMaximum=1, TargetMinimum=0, Value=@"Get Delta Ref_13".Result.Rotation.Pitch).Result
   )
   &RigUnit_ParentConstraint(
    AdvancedSettings=(InterpolationType=Average,RotationOrderForFilter=XZY)
    Child=(Type=Bone,Name="deform_lowerSleeve_l")
    Parents=((Item=(Type=Bone,Name="lowerarm_twist_01_l"),Weight=1))
    Weight=0.5
    ExecutePin=@RigVMFunction_Sequence_9.B
   )
   > &RigUnit_ParentConstraint(
    AdvancedSettings=(InterpolationType=Average,RotationOrderForFilter=XZY)
    Child=(Type=Bone,Name="deform_lowerSleeve_r")
    Parents=((Item=(Type=Bone,Name="lowerarm_twist_01_r"),Weight=1))
    Weight=0.5
   )
   > &RigUnit_ParentConstraint(
    AdvancedSettings=(InterpolationType=Average,RotationOrderForFilter=XZY)
    Child=(Type=Bone,Name="deform_upperSleeve_l")
    Parents=((Item=(Type=Bone,Name="upperarm_twist_02_l"),Weight=1))
    Weight=1
   )
   > &RigUnit_ParentConstraint(
    AdvancedSettings=(InterpolationType=Average,RotationOrderForFilter=XZY)
    Child=(Type=Bone,Name="deform_upperSleeve_r")
    Parents=((Item=(Type=Bone,Name="upperarm_twist_02_r"),Weight=1))
    Weight=1
   )
   > RigVMFunction_Sequence_5 = RigVMAggregateNode()
   Get Delta Ref_10 = RigVMFunctionReferenceNode(Item=(Type=Bone,Name="foot_l"), ExecuteContext=@RigVMFunction_Sequence_5.A)
   n3 = &RigVMFunction_MathDoubleRemap(SourceMaximum=-37, SourceMinimum=0, TargetMaximum=1, TargetMinimum=0, Value=@"Get Delta Ref_10".Result.Rotation.Yaw)
   > &RigUnit_ModifyTransforms(
    ItemToModify=((Item=(Type=Bone,Name="deform_foot_bk_l"),Transform=(Rotation=(X=0,Y=0,Z=0,W=1),Translation=(X=-1,Y=-1,Z=0),Scale3D=(X=1,Y=1,Z=1))))
    WeightMinimum=0
    Weight=@n3.Result
   )
   > &RigUnit_ModifyTransforms(
    ItemToModify=((Item=(Type=Bone,Name="base_lowerPantFr_l"),Transform=(Rotation=(X=0,Y=0,Z=0,W=1),Translation=(X=2,Y=0,Z=0),Scale3D=(X=1,Y=1,Z=1))))
    WeightMinimum=0
    Weight=@n3.Result
   )
   > &RigUnit_ModifyTransforms(
    ItemToModify=((Item=(Type=Bone,Name="base_lowerPantBk_l"),Transform=(Rotation=(X=0,Y=0,Z=0,W=1),Translation=(X=-2,Y=2,Z=0),Scale3D=(X=1,Y=1,Z=1))))
    WeightMinimum=0
    Weight=@n3.Result
   )
   Get Delta Ref_11 = RigVMFunctionReferenceNode(Item=(Type=Bone,Name="foot_r"), ExecuteContext=@RigVMFunction_Sequence_5.B)
   n4 = &RigVMFunction_MathDoubleRemap(SourceMaximum=-37, SourceMinimum=0, TargetMaximum=1, TargetMinimum=0, Value=@"Get Delta Ref_11".Result.Rotation.Yaw)
   > &RigUnit_ModifyTransforms(
    ItemToModify=((Item=(Type=Bone,Name="deform_foot_bk_r"),Transform=(Rotation=(X=0,Y=0,Z=0,W=1),Translation=(X=-2,Y=-2,Z=0),Scale3D=(X=1,Y=1,Z=1))))
    WeightMinimum=0
    Weight=@n4.Result
   )
   > &RigUnit_ModifyTransforms(
    ItemToModify=((Item=(Type=Bone,Name="base_lowerPantFr_r"),Transform=(Rotation=(X=0,Y=0,Z=0,W=1),Translation=(X=-2,Y=0,Z=0),Scale3D=(X=1,Y=1,Z=1))))
    WeightMinimum=0
    Weight=@n4.Result
   )
   > &RigUnit_ModifyTransforms(
    ItemToModify=((Item=(Type=Bone,Name="base_lowerPantBk_r"),Transform=(Rotation=(X=0,Y=0,Z=0,W=1),Translation=(X=2,Y=-2,Z=0),Scale3D=(X=1,Y=1,Z=1))))
    WeightMinimum=0
    Weight=@n4.Result
   )
   n5 = &RigVMFunction_ControlFlowBranch(ExecuteContext=@RigVMFunction_Sequence_9.C, Condition=RigVMVariableNode() {variable=Pistol}.Value)
   &RigUnit_ModifyTransforms(
    ItemToModify=((Item=(Type=Bone,Name="base_handSleeveIn_l"),Transform=(Rotation=(X=0,Y=0,Z=0.1736,W=0.9848),Translation=(X=0,Y=1,Z=0),Scale3D=(X=1,Y=1,Z=1))))
    Weight=1
    WeightMinimum=0
    ExecutePin=@n5.True
   )
   > &RigUnit_ModifyTransforms(
    ItemToModify=((Item=(Type=Bone,Name="base_handSleeveIn_r"),Transform=(Rotation=(X=0,Y=0,Z=0.1736,W=0.9848),Translation=(X=0,Y=-1,Z=0),Scale3D=(X=1,Y=1,Z=1))))
    Weight=1
    WeightMinimum=0
   )
   &RigUnit_ParentConstraint(
    AdvancedSettings=(InterpolationType=Average,RotationOrderForFilter=XZY)
    Child=(Type=Bone,Name="calf_twist_02_l")
    Parents=((Item=(Type=Bone,Name="calf_l"),Weight=1))
    Weight=1
    ExecutePin=@RigVMFunction_Sequence_1.B
   )
   > &RigUnit_ParentConstraint(
    AdvancedSettings=(InterpolationType=Average,RotationOrderForFilter=XZY)
    Child=(Type=Bone,Name="calf_twist_02_r")
    Parents=((Item=(Type=Bone,Name="calf_r"),Weight=1))
    Weight=1
   )
   > RigVMFunction_Sequence = RigVMAggregateNode()
   Get Delta Ref = RigVMFunctionReferenceNode(Item=(Type=Bone,Name="foot_l"), ExecuteContext=@RigVMFunction_Sequence.A)
   Remap = &RigVMFunction_MathDoubleRemap(SourceMaximum=60, SourceMinimum=0, TargetMaximum=1, TargetMinimum=0, Value=@"Get Delta Ref".Result.Rotation.Yaw)
   > &RigUnit_ModifyTransforms(
    ItemToModify=((Item=(Type=Bone,Name="deform_foot_bk_l"),Transform=(Rotation=(X=0,Y=0,Z=0.342,W=0.9397),Translation=(X=5,Y=6,Z=0),Scale3D=(X=1,Y=1,Z=1))))
    WeightMinimum=0
    Weight=@Remap.Result
   )
   > &RigUnit_ModifyTransforms(
    ItemToModify=((Item=(Type=Bone,Name="base_lowerPantBk_l"),Transform=(Rotation=(X=0,Y=0,Z=0,W=1),Translation=(X=6,Y=0,Z=0),Scale3D=(X=1,Y=1,Z=1))))
    WeightMinimum=0
    Weight=@Remap.Result
   )
   > &RigUnit_ModifyTransforms(
    ItemToModify=((Item=(Type=Bone,Name="base_lowerPantOut_l"),Transform=(Rotation=(X=0,Y=0,Z=0,W=1),Translation=(X=5,Y=0,Z=0),Scale3D=(X=1,Y=1,Z=1))))
    WeightMinimum=0
    Weight=@Remap.Result
   )
   > &RigUnit_ModifyTransforms(
    ItemToModify=((Item=(Type=Bone,Name="base_lowerPantIn_l"),Transform=(Rotation=(X=0,Y=0,Z=0,W=1),Translation=(X=5,Y=0,Z=0),Scale3D=(X=1,Y=1,Z=1))))
    WeightMinimum=0
    Weight=@Remap.Result
   )
   Get Delta Ref_1 = RigVMFunctionReferenceNode(Item=(Type=Bone,Name="foot_r"), ExecuteContext=@RigVMFunction_Sequence.B)
   n6 = &RigVMFunction_MathDoubleRemap(SourceMaximum=60, SourceMinimum=0, TargetMaximum=1, TargetMinimum=0, Value=@"Get Delta Ref_1".Result.Rotation.Yaw)
   > &RigUnit_ModifyTransforms(
    ItemToModify=((Item=(Type=Bone,Name="deform_foot_bk_r"),Transform=(Rotation=(X=0,Y=0,Z=0.342,W=0.9397),Translation=(X=5,Y=2,Z=0),Scale3D=(X=1,Y=1,Z=1))))
    WeightMinimum=0
    Weight=@n6.Result
   )
   > &RigUnit_ModifyTransforms(
    ItemToModify=((Item=(Type=Bone,Name="base_lowerPantBk_r"),Transform=(Rotation=(X=0,Y=0,Z=0,W=1),Translation=(X=-6,Y=0,Z=0),Scale3D=(X=1,Y=1,Z=1))))
    WeightMinimum=0
    Weight=@n6.Result
   )
   > &RigUnit_ModifyTransforms(
    ItemToModify=((Item=(Type=Bone,Name="base_lowerPantOut_r"),Transform=(Rotation=(X=0,Y=0,Z=0,W=1),Translation=(X=-5,Y=0,Z=0),Scale3D=(X=1,Y=1,Z=1))))
    WeightMinimum=0
    Weight=@n6.Result
   )
   > &RigUnit_ModifyTransforms(
    ItemToModify=((Item=(Type=Bone,Name="base_lowerPantIn_r"),Transform=(Rotation=(X=0,Y=0,Z=0,W=1),Translation=(X=-5,Y=0,Z=0),Scale3D=(X=1,Y=1,Z=1))))
    WeightMinimum=0
    Weight=@n6.Result
   )
   n7 = &RigUnit_SphericalPoseReader(
    ActiveRegionScaleFactors=(PositiveWidth=1,NegativeWidth=1,PositiveHeight=1,NegativeHeight=1)
    ActiveRegionSize=0.1
    Debug=(bDrawDebug=true,bDraw2D=false,bDrawLocalAxes=False,DebugScale=25,DebugSegments=20,DebugThickness=0.25)
    DriverAxis=(X=-1,Y=0,Z=0)
    DriverItem=(Type=Bone,Name="thigh_l")
    FalloffRegionScaleFactors=(PositiveWidth=1,NegativeWidth=1,PositiveHeight=1,NegativeHeight=1)
    FalloffSize=0.4
    FlipHeightScaling=False
    FlipWidthScaling=False
    OptionalParentItem=(Type=Bone,Name="None")
    RotationOffset=(X=120,Y=0,Z=0)
    ExecutePin=@RigVMFunction_Sequence_1.C
   )
   > n8 = &RigUnit_SphericalPoseReader(
    ActiveRegionScaleFactors=(PositiveWidth=1,NegativeWidth=1,PositiveHeight=1,NegativeHeight=1)
    ActiveRegionSize=0.1
    Debug=(bDrawDebug=True,bDraw2D=False,bDrawLocalAxes=False,DebugScale=25,DebugSegments=20,DebugThickness=0.25)
    DriverAxis=(X=1,Y=0,Z=0)
    DriverItem=(Type=Bone,Name="thigh_r")
    FalloffRegionScaleFactors=(PositiveWidth=1,NegativeWidth=1,PositiveHeight=1,NegativeHeight=1)
    FalloffSize=0.4
    FlipHeightScaling=False
    FlipWidthScaling=False
    OptionalParentItem=(Type=Bone,Name="None")
    RotationOffset=(X=-60,Y=0,Z=0)
   )
   > &RigUnit_ModifyTransforms(
    ItemToModify=((Item=(Type=Bone,Name="base_pelvis_softball"),Transform=(Rotation=(X=0,Y=0,Z=0.4226,W=0.9063),Translation=(X=-5,Y=2,Z=0),Scale3D=(X=1,Y=1,Z=1))))
    WeightMinimum=0
    Weight=&RigVMFunction_MathDoubleMul(B=@n7.OutputParam, A=@n8.OutputParam).Result
   )
   n9 = &RigUnit_SphericalPoseReader(
    ActiveRegionScaleFactors=(PositiveWidth=1,NegativeWidth=1,PositiveHeight=1,NegativeHeight=1)
    ActiveRegionSize=0.1
    Debug=(bDrawDebug=True,bDraw2D=False,bDrawLocalAxes=False,DebugScale=25,DebugSegments=20,DebugThickness=0.25)
    DriverAxis=(X=-1,Y=0,Z=0)
    DriverItem=(Type=Bone,Name="clavicle_r")
    FalloffRegionScaleFactors=(PositiveWidth=1,NegativeWidth=1,PositiveHeight=1,NegativeHeight=1)
    FalloffSize=0.3
    FlipHeightScaling=False
    FlipWidthScaling=False
    OptionalParentItem=(Type=Bone,Name="None")
    RotationOffset=(X=0,Y=140,Z=0)
    ExecutePin=@RigVMFunction_Sequence_1.D
   )
   > &RigUnit_ModifyTransforms(
    ItemToModify="((Item=(Type=Bone,Name=\"base_collar_r\"),Transform=(Rotation=(X=-0.1574,Y=-0.0734,Z=0.4162,W=0.8925),Translation=(X=-2,Y=1,Z=0),Scale3D=(X=1,Y=1,Z=1))),(Item=(Type=Bone,Name=\"base_collar_bk_r\"),Transform=(Rotation=(X=-0.0842,Y=-0.0226,Z=0.2578,W=0.9623),Tra ...(truncated)"
    WeightMinimum=0
    Weight=@n9.OutputParam
   )
   > n10 = &RigUnit_SphericalPoseReader(
    ActiveRegionScaleFactors=(PositiveWidth=1,NegativeWidth=1,PositiveHeight=1,NegativeHeight=1)
    ActiveRegionSize=0.1
    Debug=(bDrawDebug=True,bDraw2D=False,bDrawLocalAxes=False,DebugScale=25,DebugSegments=20,DebugThickness=0.25)
    DriverAxis=(X=1,Y=0,Z=0)
    DriverItem=(Type=Bone,Name="clavicle_l")
    FalloffRegionScaleFactors=(PositiveWidth=1,NegativeWidth=1,PositiveHeight=1,NegativeHeight=1)
    FalloffSize=0.3
    FlipHeightScaling=False
    FlipWidthScaling=False
    OptionalParentItem=(Type=Bone,Name="None")
    RotationOffset=(X=0,Y=-40,Z=0)
   )
   > &RigUnit_ModifyTransforms(
    ItemToModify="((Item=(Type=Bone,Name=\"base_collar_l\"),Transform=(Rotation=(X=-0.1574,Y=-0.0734,Z=0.4162,W=0.8925),Translation=(X=-2,Y=1,Z=0),Scale3D=(X=1,Y=1,Z=1))),(Item=(Type=Bone,Name=\"base_collar_bk_l\"),Transform=(Rotation=(X=0.0842,Y=0.0226,Z=0.2578,W=0.9623),Trans ...(truncated)"
    WeightMinimum=0
    Weight=@n10.OutputParam
   )
   Multiply = &RigVMFunction_MathDoubleMul(A=@n9.OutputParam, B=@n10.OutputParam)
   > n11 = &RigUnit_ModifyTransforms(
    ItemToModify="((Item=(Type=Bone,Name=\"base_collar_bk\"),Transform=(Rotation=(X=0,Y=0,Z=0.0872,W=0.9962),Translation=(X=0,Y=0,Z=0),Scale3D=(X=1,Y=1,Z=1))),(Item=(Type=Bone,Name=\"base_collar_front\"),Transform=(Rotation=(X=0,Y=0,Z=0,W=1),Translation=(X=-1,Y=2,Z=0),Scale3D=( ...(truncated)"
    WeightMinimum=0
    Weight=@Multiply.Result
   )
   n12 = &RigVMFunction_Sequence(ExecuteContext=@n11.ExecutePin)
   n13 = &RigVMFunction_ControlFlowBranch(ExecuteContext=@n12.A, Condition=RigVMVariableNode() {variable=NotRBAN}.Value)
   n14 = &RigVMFunction_ControlFlowBranch(ExecuteContext=@n13.False, Condition=RigVMVariableNode() {variable=Emote}.Value)
   n15 = &RigUnit_SphericalPoseReader(
    ActiveRegionScaleFactors=(PositiveWidth=2,NegativeWidth=1,PositiveHeight=1,NegativeHeight=1)
    ActiveRegionSize=0.1
    Debug=(bDrawDebug=True,bDraw2D=False,bDrawLocalAxes=False,DebugScale=25,DebugSegments=20,DebugThickness=0.25)
    DriverAxis=(X=-1,Y=0,Z=0)
    DriverItem=(Type=Bone,Name="upperarm_r")
    FalloffRegionScaleFactors=(PositiveWidth=1,NegativeWidth=1,PositiveHeight=1,NegativeHeight=1)
    FalloffSize=0.35
    FlipHeightScaling=False
    FlipWidthScaling=False
    OptionalParentItem=(Type=Bone,Name="None")
    RotationOffset=(X=-150,Y=40,Z=0)
    ExecutePin=@n14.False
   )
   > &RigUnit_ModifyTransforms(
    ItemToModify="((Item=(Type=Bone,Name=\"base_collar_r\"),Transform=(Rotation=(X=-0.1853,Y=-0.509,Z=0.1853,W=0.8199),Translation=(X=0,Y=0,Z=0),Scale3D=(X=1,Y=1,Z=1))),(Item=(Type=Bone,Name=\"base_collar_bk_r\"),Transform=(Rotation=(X=0,Y=0,Z=0,W=1),Translation=(X=0,Y=0,Z=0),S ...(truncated)"
    WeightMinimum=0
    Weight=@n15.OutputParam
   )
   > n16 = &RigUnit_SphericalPoseReader(
    ActiveRegionScaleFactors=(PositiveWidth=1,NegativeWidth=1,PositiveHeight=1,NegativeHeight=1)
    ActiveRegionSize=0.1
    Debug=(bDrawDebug=True,bDraw2D=False,bDrawLocalAxes=False,DebugScale=25,DebugSegments=20,DebugThickness=0.25)
    DriverAxis=(X=1,Y=0,Z=0)
    DriverItem=(Type=Bone,Name="upperarm_l")
    FalloffRegionScaleFactors=(PositiveWidth=1,NegativeWidth=1,PositiveHeight=1,NegativeHeight=1)
    FalloffSize=0.35
    FlipHeightScaling=False
    FlipWidthScaling=False
    OptionalParentItem=(Type=Bone,Name="None")
    RotationOffset=(X=40,Y=-40,Z=0)
   )
   > &RigUnit_ModifyTransforms(
    ItemToModify="((Item=(Type=Bone,Name=\"base_collar_l\"),Transform=(Rotation=(X=-0.1422,Y=0.5649,Z=0.0996,W=0.8067),Translation=(X=0,Y=0,Z=0),Scale3D=(X=1,Y=1,Z=1))),(Item=(Type=Bone,Name=\"base_collar_bk_l\"),Transform=(Rotation=(X=0,Y=0,Z=0,W=1),Translation=(X=0,Y=0,Z=0),S ...(truncated)"
    WeightMinimum=0
    Weight=@n16.OutputParam
   )
   > &RigUnit_ModifyTransforms(
    ItemToModify="((Item=(Type=Bone,Name=\"base_collar_bk\"),Transform=(Rotation=(X=0,Y=0,Z=0.1736,W=0.9848),Translation=(X=0,Y=0,Z=0),Scale3D=(X=1,Y=1,Z=1))),(Item=(Type=Bone,Name=\"base_collar_front\"),Transform=(Rotation=(X=0,Y=0,Z=0.1736,W=0.9848),Translation=(X=-2,Y=2,Z=0) ...(truncated)"
    WeightMinimum=0
    Weight=&RigVMFunction_MathDoubleMul(B=@n16.OutputParam, A=@n15.OutputParam).Result
   )
   n17 = &RigUnit_SphericalPoseReader(
    ActiveRegionScaleFactors=(PositiveWidth=2,NegativeWidth=1,PositiveHeight=1,NegativeHeight=1)
    ActiveRegionSize=0.1
    Debug=(bDrawDebug=True,bDraw2D=False,bDrawLocalAxes=False,DebugScale=25,DebugSegments=20,DebugThickness=0.25)
    DriverAxis=(X=-1,Y=0,Z=0)
    DriverItem=(Type=Bone,Name="upperarm_r")
    FalloffRegionScaleFactors=(PositiveWidth=1,NegativeWidth=1,PositiveHeight=1,NegativeHeight=1)
    FalloffSize=0.35
    FlipHeightScaling=False
    FlipWidthScaling=False
    OptionalParentItem=(Type=Bone,Name="None")
    RotationOffset=(X=-150,Y=40,Z=0)
    ExecutePin=@n13.True
   )
   > &RigUnit_ModifyTransforms(
    ItemToModify="((Item=(Type=Bone,Name=\"base_collar_r\"),Transform=(Rotation=(X=-0.3213,Y=-0.1398,Z=0.1332,W=0.9271),Translation=(X=0,Y=0,Z=0),Scale3D=(X=1,Y=1,Z=1))),(Item=(Type=Bone,Name=\"base_collar_bk_r\"),Transform=(Rotation=(X=0,Y=0,Z=0,W=1),Translation=(X=0,Y=0,Z=0), ...(truncated)"
    WeightMinimum=0
    Weight=@n17.OutputParam
   )
   > n18 = &RigUnit_SphericalPoseReader(
    ActiveRegionScaleFactors=(PositiveWidth=1,NegativeWidth=1,PositiveHeight=1,NegativeHeight=1)
    ActiveRegionSize=0.1
    Debug=(bDrawDebug=True,bDraw2D=False,bDrawLocalAxes=False,DebugScale=25,DebugSegments=20,DebugThickness=0.25)
    DriverAxis=(X=1,Y=0,Z=0)
    DriverItem=(Type=Bone,Name="upperarm_l")
    FalloffRegionScaleFactors=(PositiveWidth=1,NegativeWidth=1,PositiveHeight=1,NegativeHeight=1)
    FalloffSize=0.35
    FlipHeightScaling=False
    FlipWidthScaling=False
    OptionalParentItem=(Type=Bone,Name="None")
    RotationOffset=(X=40,Y=-40,Z=0)
   )
   > &RigUnit_ModifyTransforms(
    ItemToModify="((Item=(Type=Bone,Name=\"base_collar_l\"),Transform=(Rotation=(X=0.1332,Y=0.1398,Z=0.3213,W=0.9271),Translation=(X=0,Y=0,Z=0),Scale3D=(X=1,Y=1,Z=1))),(Item=(Type=Bone,Name=\"base_collar_bk_l\"),Transform=(Rotation=(X=0,Y=0,Z=0,W=1),Translation=(X=0,Y=0,Z=0),Sc ...(truncated)"
    WeightMinimum=0
    Weight=@n18.OutputParam
   )
   RigVMFunction_ControlFlowBranch = &RigVMFunction_ControlFlowBranch(ExecuteContext=@RigVMFunction_Sequence_1.E, Condition=RigVMVariableNode() {variable=Melee}.Value)
   RigVMFunction_Sequence_3 = RigVMAggregateNode(ExecuteContext=@RigVMFunction_ControlFlowBranch.True)
   n19 = &RigVMFunction_Sequence(ExecuteContext=@RigVMFunction_Sequence_3.A)
   Get Delta Ref_2 = RigVMFunctionReferenceNode(Item=(Type=Bone,Name="index_01_r"), ExecuteContext=@n19.A)
   > &RigUnit_ModifyTransforms(
    ItemToModify=((Item=(Type=Bone,Name="base_handSleeveBk_r"),Transform=(Rotation=(X=-0.212,Y=-0.554,Z=-0.1485,W=0.7912),Translation=(X=2,Y=1,Z=1),Scale3D=(X=1,Y=1,Z=1))))
    WeightMinimum=0
    Weight=&RigVMFunction_MathDoubleRemap(SourceMaximum=-35, SourceMinimum=-12, TargetMaximum=1, TargetMinimum=0, Value=@"Get Delta Ref_2".Result.Rotation.Yaw).Result
   )
   Get Delta Ref_3 = RigVMFunctionReferenceNode(Item=(Type=Bone,Name="index_01_r"), ExecuteContext=@n19.B)
   > &RigUnit_ModifyTransforms(
    ItemToModify=((Item=(Type=Bone,Name="base_handSleeveFr_r"),Transform=(Rotation=(X=0.25,Y=0.25,Z=-0.067,W=0.933),Translation=(X=1,Y=0,Z=0),Scale3D=(X=1,Y=1,Z=1))))
    WeightMinimum=0
    Weight=&RigVMFunction_MathDoubleRemap(SourceMaximum=-20, SourceMinimum=4, TargetMaximum=1, TargetMinimum=0, Value=@"Get Delta Ref_3".Result.Rotation.Yaw).Result
   )
   n20 = &RigVMFunction_Sequence(ExecuteContext=@RigVMFunction_Sequence_3.B)
   Get Delta Ref_4 = RigVMFunctionReferenceNode(Item=(Type=Bone,Name="index_01_l"), ExecuteContext=@n20.A)
   > &RigUnit_ModifyTransforms(
    ItemToModify=((Item=(Type=Bone,Name="base_handSleeveBk_l"),Transform=(Rotation=(X=-0.212,Y=-0.554,Z=-0.1485,W=0.7912),Translation=(X=-2,Y=-1,Z=1),Scale3D=(X=1,Y=1,Z=1))))
    WeightMinimum=0
    Weight=&RigVMFunction_MathDoubleRemap(SourceMaximum=-45, SourceMinimum=-35, TargetMaximum=1, TargetMinimum=0, Value=@"Get Delta Ref_4".Result.Rotation.Yaw).Result
   )
   Get Delta Ref_5 = RigVMFunctionReferenceNode(Item=(Type=Bone,Name="pinky_03_l"), ExecuteContext=@n20.B)
   > &RigUnit_ModifyTransforms(
    ItemToModify=((Item=(Type=Bone,Name="base_handSleeveFr_l"),Transform=(Rotation=(X=0.1022,Y=0.36,Z=0.1022,W=0.9217),Translation=(X=1,Y=0,Z=0),Scale3D=(X=1,Y=1,Z=1))))
    WeightMinimum=0
    Weight=&RigVMFunction_MathDoubleRemap(SourceMaximum=-35, SourceMinimum=-12, TargetMaximum=1, TargetMinimum=0, Value=@"Get Delta Ref_5".Result.Rotation.Yaw).Result
   )
   n21 = &RigVMFunction_Sequence(ExecuteContext=@RigVMFunction_Sequence_3.C)
   Get Delta Ref_6 = RigVMFunctionReferenceNode(Item=(Type=Bone,Name="thumb_01_r"), ExecuteContext=@n21.A)
   > &RigUnit_ModifyTransforms(
    ItemToModify=((Item=(Type=Bone,Name="base_handSleeveFr_r"),Transform=(Rotation=(X=0.1574,Y=0.4162,Z=-0.0734,W=0.8925),Translation=(X=2,Y=0,Z=3),Scale3D=(X=1,Y=1,Z=1))))
    WeightMinimum=0
    Weight=&RigVMFunction_MathDoubleRemap(SourceMaximum=12, SourceMinimum=5, TargetMaximum=1, TargetMinimum=0, Value=@"Get Delta Ref_6".Result.Rotation.Yaw).Result
   )
   Get Delta Ref_7 = RigVMFunctionReferenceNode(Item=(Type=Bone,Name="thumb_01_l"), ExecuteContext=@n21.B)
   > &RigUnit_ModifyTransforms(
    ItemToModify=((Item=(Type=Bone,Name="base_handSleeveFr_l"),Transform=(Rotation=(X=0.1574,Y=0.4162,Z=-0.0734,W=0.8925),Translation=(X=-2,Y=0,Z=-3),Scale3D=(X=1,Y=1,Z=1))))
    WeightMinimum=0
    Weight=&RigVMFunction_MathDoubleRemap(SourceMaximum=12, SourceMinimum=5, TargetMaximum=1, TargetMinimum=0, Value=@"Get Delta Ref_7".Result.Rotation.Yaw).Result
   )
   n22 = &RigVMFunction_ControlFlowBranch(ExecuteContext=@RigVMFunction_Sequence_1.F, Condition=RigVMVariableNode() {variable=DBNO}.Value)
   n23 = &RigVMFunction_Sequence(ExecuteContext=@n22.True)
   Get Delta Ref_8 = RigVMFunctionReferenceNode(Item=(Type=Bone,Name="hand_r"), ExecuteContext=@n23.A)
   n24 = &RigVMFunction_MathDoubleRemap(SourceMaximum=45, SourceMinimum=20, TargetMaximum=1, TargetMinimum=0, Value=@"Get Delta Ref_8".Result.Rotation.Yaw)
   > &RigUnit_ModifyTransforms(
    ItemToModify=((Item=(Type=Bone,Name="base_handSleeveOut_r"),Transform=(Rotation=(X=0,Y=0,Z=-0.1736,W=0.9848),Translation=(X=0,Y=1,Z=0),Scale3D=(X=1,Y=1,Z=1))))
    WeightMinimum=0
    Weight=@n24.Result
   )
   > &RigUnit_ModifyTransforms(
    ItemToModify=((Item=(Type=Bone,Name="base_handSleeveIn_r"),Transform=(Rotation=(X=0,Y=0,Z=0.2588,W=0.9659),Translation=(X=0,Y=-2,Z=0),Scale3D=(X=1,Y=1,Z=1))))
    WeightMinimum=0
    Weight=@n24.Result
   )
   > &RigUnit_ModifyTransforms(
    ItemToModify=((Item=(Type=Bone,Name="base_handSleeveFr_r"),Transform=(Rotation=(X=0.067,Y=0.25,Z=-0.25,W=0.933),Translation=(X=0,Y=0,Z=0),Scale3D=(X=1,Y=0.5,Z=1))))
    WeightMinimum=0
    Weight=@n24.Result
   )
   > &RigUnit_ModifyTransforms(
    ItemToModify=((Item=(Type=Bone,Name="base_handSleeveBk_r"),Transform=(Rotation=(X=-0.0449,Y=-0.2549,Z=-0.1677,W=0.9513),Translation=(X=0,Y=0,Z=0),Scale3D=(X=1,Y=0.5,Z=1))))
    WeightMinimum=0
    Weight=@n24.Result
   )
   Get Delta Ref_9 = RigVMFunctionReferenceNode(Item=(Type=Bone,Name="hand_l"), ExecuteContext=@n23.B)
   n25 = &RigVMFunction_MathDoubleRemap(SourceMaximum=45, SourceMinimum=20, TargetMaximum=1, TargetMinimum=0, Value=@"Get Delta Ref_9".Result.Rotation.Yaw)
   > &RigUnit_ModifyTransforms(
    ItemToModify=((Item=(Type=Bone,Name="base_handSleeveOut_l"),Transform=(Rotation=(X=0,Y=0,Z=-0.1736,W=0.9848),Translation=(X=0,Y=-1,Z=0),Scale3D=(X=1,Y=1,Z=1))))
    WeightMinimum=0
    Weight=@n25.Result
   )
   > &RigUnit_ModifyTransforms(
    ItemToModify=((Item=(Type=Bone,Name="base_handSleeveIn_l"),Transform=(Rotation=(X=0,Y=0,Z=0.2588,W=0.9659),Translation=(X=0,Y=2,Z=0),Scale3D=(X=1,Y=1,Z=1))))
    WeightMinimum=0
    Weight=@n25.Result
   )
   > &RigUnit_ModifyTransforms(
    ItemToModify=((Item=(Type=Bone,Name="base_handSleeveFr_l"),Transform=(Rotation=(X=0.067,Y=0.25,Z=-0.25,W=0.933),Translation=(X=0,Y=0,Z=0),Scale3D=(X=1,Y=1,Z=1))))
    WeightMinimum=0
    Weight=@n25.Result
   )
   > &RigUnit_ModifyTransforms(
    ItemToModify=((Item=(Type=Bone,Name="base_handSleeveBk_l"),Transform=(Rotation=(X=-0.0449,Y=-0.2549,Z=-0.1677,W=0.9513),Translation=(X=0,Y=0,Z=0),Scale3D=(X=1,Y=1,Z=1))))
    WeightMinimum=0
    Weight=@n25.Result
   )
   n26 = &RigVMFunction_Sequence(ExecuteContext=@RigVMFunction_Sequence_1.G)
   n27 = &RigVMFunction_ControlFlowBranch(ExecuteContext=@n26.A, Condition=RigVMVariableNode() {variable=ArmsHidden}.Value)
   &RigUnit_ModifyTransforms(
    ItemToModify=((Item=(Type=Bone,Name="deform_lowerSleeve_r"),Transform=(Rotation=(X=0,Y=0,Z=0,W=1),Translation=(X=0,Y=0,Z=0),Scale3D=(X=0.1,Y=0.1,Z=0.1))))
    Weight=1
    WeightMinimum=0
    ExecutePin=@n27.True
   )
   > &RigUnit_ModifyTransforms(
    ItemToModify=((Item=(Type=Bone,Name="deform_lowerSleeve_l"),Transform=(Rotation=(X=0,Y=0,Z=0,W=1),Translation=(X=0,Y=0,Z=0),Scale3D=(X=0.1,Y=0.1,Z=0.1))))
    Weight=1
    WeightMinimum=0
   )
   n28 = &RigVMFunction_ControlFlowBranch(ExecuteContext=@n26.B, Condition=RigVMVariableNode() {variable=LegsHidden}.Value)
   &RigUnit_ModifyTransforms(
    ItemToModify=((Item=(Type=Bone,Name="deform_lowerPant_r"),Transform=(Rotation=(X=0,Y=0,Z=0,W=1),Translation=(X=0,Y=0,Z=0),Scale3D=(X=0.1,Y=0.1,Z=0.1))))
    Weight=1
    WeightMinimum=0
    ExecutePin=@n28.True
   )
   > &RigUnit_ModifyTransforms(
    ItemToModify=((Item=(Type=Bone,Name="deform_lowerPant_l"),Transform=(Rotation=(X=0,Y=0,Z=0,W=1),Translation=(X=0,Y=0,Z=0),Scale3D=(X=0.1,Y=0.1,Z=0.1))))
    Weight=1
    WeightMinimum=0
   )
  ]}
]
hierarchy [
 root {loc=<0 0 0>, rot=<0 0 0 1>}
 pelvis {parent=root, loc=<0 -0.7071 93.1191>, rot=<0.0127 0.707 -0.0127 -0.707>}
 spine_01 {parent=pelvis, loc=<6.7494 -0.6388 0>, rot=<0 0 0.034 -0.9994>}
 spine_02 {parent=spine_01, loc=<7.2518 -0.035 0>, rot=<0 0 0.0238 -0.9997>}
 spine_03 {parent=spine_02, loc=<6.0104 0 0>, rot=<0 0 0.0149 -0.9999>}
 spine_04 {parent=spine_03, loc=<8.1024 0 0>, rot=<0 0 -0.0132 -0.9999>}
 spine_05 {parent=spine_04, loc=<10.0146 0 0>, rot=<0 0 -0.0015 -1>}
 neck_01 {parent=spine_05, loc=<11.3138 0.1529 0>, rot=<0 0 -0.2003 -0.9797>}
 neck_02 {parent=neck_01, loc=<4.5627 0 0>, rot=<0 0 0.0504 -0.9987>}
 head {parent=neck_02, loc=<5.3914 0 0>, rot=<0 0 0.115 -0.9934>}
 attach_cape {parent=spine_05, loc=<9.017 -11.714 0>, rot=<0.0616 -0.7044 -0.0616 -0.7044>}
 attach_backpack {parent=spine_05, loc=<-0.123 -13.473 0>, rot=<-0.0093 -0.707 0.0093 -0.707>}
 clavicle_l {parent=spine_05, loc=<4.3575 -0.1466 -2.3223>, rot=<0.0465 -0.7289 -0.009 -0.6829>}
 upperarm_l {parent=clavicle_l, loc=<15.5111 0 0>, rot=<0.0226 -0.3234 -0.0088 -0.9459>}
 lowerarm_l {parent=upperarm_l, loc=<26.6514 0 0>, rot=<0.0391 -0.0099 -0.2453 -0.9686>}
 hand_l {parent=lowerarm_l, loc=<22.3491 0 0>, rot=<-0.7081 -0.0507 0.0975 -0.6975>}
 thumb_01_l {parent=hand_l, loc=<1.3118 -0.3191 -2.459>, rot=<0.6546 -0.4042 0.0664 -0.6354>}
 thumb_02_l {parent=thumb_01_l, loc=<5.9336 -0.0448 0.0401>, rot=<-0.0037 0.0213 0.298 -0.9543>}
 thumb_03_l {parent=thumb_02_l, loc=<2.2223 0.3682 -0.0907>, rot=<0.0118 -0.013 0.0066 -0.9998>}
 index_metacarpal_l {parent=hand_l, loc=<1.6535 0.0183 -2.96>, rot=<0.0158 -0.0124 -0.0173 -0.9996>}
 index_01_l {parent=index_metacarpal_l, loc=<8.3008 0.0357 -0.0524>, rot=<-0.0044 0.0485 0.0924 -0.9945>}
 index_02_l {parent=index_01_l, loc=<3.3587 -0.3004 0.0452>, rot=<-0.002 0.0027 0.3225 -0.9466>}
 index_03_l {parent=index_02_l, loc=<2.5578 -0.123 -0.0252>, rot=<-0.002 0.0067 0.2558 -0.9667>}
 middle_metacarpal_l {parent=hand_l, loc=<2.1856 -0.1725 -0.5984>, rot=<0.006 0.0168 -0.0329 -0.9993>}
 middle_01_l {parent=middle_metacarpal_l, loc=<7.3275 -0.2447 0.0129>, rot=<-0.0208 0.0042 0.0858 -0.9961>}
 middle_02_l {parent=middle_01_l, loc=<4.0123 0.1401 0.1175>, rot=<-0.0218 0.0258 0.3512 -0.9357>}
 middle_03_l {parent=middle_02_l, loc=<2.6077 0.0534 0.1373>, rot=<0.0082 0.0188 0.1708 -0.9851>}
 ring_metacarpal_l {parent=hand_l, loc=<2.2074 -0.3625 1.3638>, rot=<-0.0284 0.0722 -0.0458 -0.9959>}
 ring_01_l {parent=ring_metacarpal_l, loc=<7.0295 -0.6197 -0.1446>, rot=<-0.0489 -0.0252 0.1892 -0.9804>}
 ring_02_l {parent=ring_01_l, loc=<3.6174 0.4647 -0.1027>, rot=<0 0 0.2806 -0.9598>}
 ring_03_l {parent=ring_02_l, loc=<2.4152 0.114 0.0199>, rot=<0 0 0.1189 -0.9929>}
 pinky_metacarpal_l {parent=hand_l, loc=<1.7616 -0.9753 3.3814>, rot=<-0.0397 0.0814 -0.069 -0.9935>}
 pinky_01_l {parent=pinky_metacarpal_l, loc=<7.1709 -0.4831 -0.1094>, rot=<-0.0654 -0.0035 0.2242 -0.9723>}
 pinky_02_l {parent=pinky_01_l, loc=<2.8274 0.0893 -0.1043>, rot=<0.0037 0.0011 0.2854 -0.9584>}
 pinky_03_l {parent=pinky_02_l, loc=<2.2046 0.0307 -0.0078>, rot=<-0.0002 0.0098 0.1445 -0.9895>}
 deform_wrist_dn_base_l {parent=hand_l, loc=<0 -2.2342 0>, rot=<0 0 0 -1>}
 deform_wrist_dn_l {parent=deform_wrist_dn_base_l, loc=<0 0 0>, rot=<0 0 0 1>}
 deform_wrist_up_base_l {parent=hand_l, loc=<0 2.2682 0>, rot=<0 0 0 -1>}
 deform_wrist_up_l {parent=deform_wrist_up_base_l, loc=<0 0 0>, rot=<0 0 0 1>}
 deform_handSleeve_l {parent=hand_l, loc=<0 0 0>, rot=<0 0 0 1>}
 base_handSleeveBk_l {parent=deform_handSleeve_l, loc=<4.2913 -0.9254 5.8565>, rot=<0 0 0 1>}
 dyn_handSleeveBk_l_01 {parent=base_handSleeveBk_l, loc=<0 0 0>, rot=<0 0 0 1>}
 base_handSleeveFr_l {parent=deform_handSleeve_l, loc=<5.0627 0 -6.157>, rot=<0 0 0 1>}
 dyn_handSleeveFr_l_01 {parent=base_handSleeveFr_l, loc=<0 0 0>, rot=<0 0 0 -1>}
 base_handSleeveOut_l {parent=deform_handSleeve_l, loc=<3.4696 3.0625 0>, rot=<0 0 0 1>}
 dyn_handSleeveOut_l_01 {parent=base_handSleeveOut_l, loc=<0 0 0>, rot=<0 0 0 1>}
 base_handSleeveIn_l {parent=deform_handSleeve_l, loc=<3.8011 -7.032 -1.1084>, rot=<0 0 0 1>}
 dyn_handSleeveIn_l_01 {parent=base_handSleeveIn_l, loc=<0 0 0>, rot=<0 0 0 -1>}
 lowerarm_twist_02_l {parent=lowerarm_l, loc=<9.6199 0 0>, rot=<0 0 0 1>}
 lowerarm_twist_01_l {parent=lowerarm_l, loc=<17.8549 0 0>, rot=<0 0 0 1>}
 deform_elbow_in_base_l {parent=lowerarm_l, loc=<0.2112 5.259 0.8274>, rot=<0 0 0 1>}
 deform_elbow_in_l {parent=deform_elbow_in_base_l, loc=<0 0 0>, rot=<0 0 0 1>}
 deform_elbow_out_base_l {parent=lowerarm_l, loc=<0 -4 0>, rot=<0 0 0 1>}
 deform_elbow_out_l {parent=deform_elbow_out_base_l, loc=<0 0 0>, rot=<0 0 0 1>}
 deform_lowerSleeve_l {parent=lowerarm_l, loc=<0 0 0>, rot=<0 0 0 1>}
 base_lowerSleeveBk_l {parent=deform_lowerSleeve_l, loc=<14.7487 -6.7687 -3.2234>, rot=<0 0 0 1>}
 dyn_lowerSleeveBk_l_01 {parent=base_lowerSleeveBk_l, loc=<0 0 0>, rot=<0 0 0 1>}
 base_lowerSleeveFr_l {parent=deform_lowerSleeve_l, loc=<15.7826 5.1382 0.0002>, rot=<0 0 0 1>}
 dyn_lowerSleeveFr_l_01 {parent=base_lowerSleeveFr_l, loc=<0 0 0>, rot=<0 0 0 -1>}
 base_lowerSleeveOut_l {parent=deform_lowerSleeve_l, loc=<16.3323 0.3778 5.007>, rot=<0 0 0 1>}
 dyn_lowerSleeveOut_l_01 {parent=base_lowerSleeveOut_l, loc=<0 0 0>, rot=<0 0 0 1>}
 base_lowerSleeveIn_l {parent=deform_lowerSleeve_l, loc=<15.4468 0.3542 -7.4647>, rot=<0 0 0 1>}
 dyn_lowerSleeveIn_l_01 {parent=base_lowerSleeveIn_l, loc=<0 0 0>, rot=<0 0 0 -1>}
 upperarm_twist_01_l {parent=upperarm_l, loc=<0.0041 0 0>, rot=<0 0 0 1>}
 deform_upperarm_bk_base_l {parent=upperarm_twist_01_l, loc=<1.0813 -7.2193 -1.9453>, rot=<0 0 0 1>}
 deform_upperarm_bk_l {parent=deform_upperarm_bk_base_l, loc=<0 0 0>, rot=<0 0 0 1>}
 deform_upperarm_fr_base_l {parent=upperarm_twist_01_l, loc=<-0.3691 7.0506 -2.765>, rot=<0 0 0 1>}
 deform_upperarm_fr_l {parent=deform_upperarm_fr_base_l, loc=<0 0 0>, rot=<0 0 0 1>}
 deform_upperarm_up_base_l {parent=upperarm_twist_01_l, loc=<-3.2216 -1.1391 3.3311>, rot=<0 0 0 1>}
 deform_upperarm_up_l {parent=deform_upperarm_up_base_l, loc=<0 0 0>, rot=<0 0 0 1>}
 upperarm_twist_02_l {parent=upperarm_l, loc=<13.7941 0 0>, rot=<0 0 0 1>}
 deform_upperSleeve_l {parent=upperarm_l, loc=<0 0 0>, rot=<0 0 0 1>}
 base_upperSleeveFr_l {parent=deform_upperSleeve_l, loc=<6.6016 6.6159 0.7981>, rot=<-0.0745 -0.0665 0.011 -0.9949>}
 dyn_upperSleeveFr_l_01 {parent=base_upperSleeveFr_l, loc=<0 0 0>, rot=<0 0 0 -1>}
 base_upperSleeveIn_l {parent=deform_upperSleeve_l, loc=<7.7618 0.5617 -7.3362>, rot=<-0.0643 -0.0543 0.0237 -0.9962>}
 dyn_upperSleeveIn_l_01 {parent=base_upperSleeveIn_l, loc=<0 0 0>, rot=<0 0 0 1>}
 base_upperSleeveOut_l {parent=deform_upperSleeve_l, loc=<7.2389 0.1188 2.6931>, rot=<-0.1116 -0.0433 -0.0137 -0.9927>}
 dyn_upperSleeveOut_l_01 {parent=base_upperSleeveOut_l, loc=<0 0 0>, rot=<0 0 0 1>}
 base_upperSleeveBk_l {parent=deform_upperSleeve_l, loc=<7.6197 -6.3259 0.2651>, rot=<-0.0942 -0.051 0.0027 -0.9942>}
 dyn_upperSleeveBk_l_01 {parent=base_upperSleeveBk_l, loc=<0 0 0>, rot=<0 0 0 1>}
 clavicle_r {parent=spine_05, loc=<4.3575 -0.1466 2.3223>, rot=<0.7289 0.0465 -0.6829 0.009>}
 upperarm_r {parent=clavicle_r, loc=<-15.5111 0 0>, rot=<0.0226 -0.3234 -0.0088 -0.9459>}
 lowerarm_r {parent=upperarm_r, loc=<-26.6515 0 0>, rot=<0.0391 -0.0099 -0.2453 -0.9686>}
 hand_r {parent=lowerarm_r, loc=<-22.3492 0 0>, rot=<-0.7081 -0.0507 0.0975 -0.6975>}
 thumb_01_r {parent=hand_r, loc=<-1.3119 0.3192 2.459>, rot=<0.6546 -0.4042 0.0664 -0.6354>}
 thumb_02_r {parent=thumb_01_r, loc=<-5.9335 0.0447 -0.0401>, rot=<-0.0037 0.0213 0.298 -0.9543>}
 thumb_03_r {parent=thumb_02_r, loc=<-2.2223 -0.3682 0.0907>, rot=<0.0118 -0.013 0.0066 -0.9998>}
 index_metacarpal_r {parent=hand_r, loc=<-1.6534 -0.0181 2.9601>, rot=<0.0158 -0.0124 -0.0173 -0.9996>}
 index_01_r {parent=index_metacarpal_r, loc=<-8.3008 -0.0357 0.0524>, rot=<-0.0044 0.0485 0.0924 -0.9945>}
 index_02_r {parent=index_01_r, loc=<-3.3589 0.3003 -0.0453>, rot=<-0.002 0.0027 0.3225 -0.9466>}
 index_03_r {parent=index_02_r, loc=<-2.5578 0.1231 0.0251>, rot=<-0.002 0.0067 0.2558 -0.9667>}
 middle_metacarpal_r {parent=hand_r, loc=<-2.1856 0.1726 0.5984>, rot=<0.006 0.0168 -0.0329 -0.9993>}
 middle_01_r {parent=middle_metacarpal_r, loc=<-7.3273 0.2447 -0.0129>, rot=<-0.0208 0.0042 0.0858 -0.9961>}
 middle_02_r {parent=middle_01_r, loc=<-4.0142 -0.1411 -0.07>, rot=<-0.0217 0.0259 0.3512 -0.9357>}
 middle_03_r {parent=middle_02_r, loc=<-2.6074 -0.0534 -0.1312>, rot=<0.0083 0.0188 0.1708 -0.9851>}
 ring_metacarpal_r {parent=hand_r, loc=<-2.2073 0.3627 -1.3638>, rot=<-0.0284 0.0722 -0.0458 -0.9959>}
 ring_01_r {parent=ring_metacarpal_r, loc=<-7.0297 0.6197 0.1446>, rot=<-0.0489 -0.0252 0.1892 -0.9804>}
 ring_02_r {parent=ring_01_r, loc=<-3.6171 -0.4648 0.1027>, rot=<0 0 0.2806 -0.9598>}
 ring_03_r {parent=ring_02_r, loc=<-2.4153 -0.1141 -0.0199>, rot=<0 0 0.1189 -0.9929>}
 pinky_metacarpal_r {parent=hand_r, loc=<-1.7617 0.9753 -3.3814>, rot=<-0.0397 0.0814 -0.069 -0.9935>}
 pinky_01_r {parent=pinky_metacarpal_r, loc=<-7.1707 0.4832 0.1094>, rot=<-0.0654 -0.0035 0.2242 -0.9723>}
 pinky_02_r {parent=pinky_01_r, loc=<-2.8274 -0.0892 0.1043>, rot=<0.0037 0.0011 0.2854 -0.9584>}
 pinky_03_r {parent=pinky_02_r, loc=<-2.2045 -0.031 0.0078>, rot=<-0.0002 0.0098 0.1445 -0.9895>}
 deform_wrist_dn_base_r {parent=hand_r, loc=<-0.0004 2.2345 -0.0001>, rot=<0 0 0 -1>}
 deform_wrist_dn_r {parent=deform_wrist_dn_base_r, loc=<0 0 0>, rot=<0 0 0 1>}
 deform_wrist_up_base_r {parent=hand_r, loc=<0.0003 -2.2684 0.0001>, rot=<0 0 0 -1>}
 deform_wrist_up_r {parent=deform_wrist_up_base_r, loc=<0 0 0>, rot=<0 0 0 1>}
 deform_handSleeve_r {parent=hand_r, loc=<0 0 0>, rot=<0 0 0 1>}
 base_handSleeveBk_r {parent=deform_handSleeve_r, loc=<-4.2913 0.9254 -5.8565>, rot=<0 0 0 1>}
 dyn_handSleeveBk_r_01 {parent=base_handSleeveBk_r, loc=<0 0 0>, rot=<0 0 0 1>}
 base_handSleeveFr_r {parent=deform_handSleeve_r, loc=<-5.0627 0 6.157>, rot=<0 0 0 1>}
 dyn_handSleeveFr_r_01 {parent=base_handSleeveFr_r, loc=<0 0 0>, rot=<0 0 0 -1>}
 base_handSleeveOut_r {parent=deform_handSleeve_r, loc=<-3.4696 -3.0625 0>, rot=<0 0 0 1>}
 dyn_handSleeveOut_r_01 {parent=base_handSleeveOut_r, loc=<0 0 0>, rot=<0 0 0 1>}
 base_handSleeveIn_r {parent=deform_handSleeve_r, loc=<-3.8011 7.032 1.1084>, rot=<0 0 0 1>}
 dyn_handSleeveIn_r_01 {parent=base_handSleeveIn_r, loc=<0 0 0>, rot=<0 0 0 1>}
 lowerarm_twist_02_r {parent=lowerarm_r, loc=<-9.6199 0 0>, rot=<0 0 0 1>}
 lowerarm_twist_01_r {parent=lowerarm_r, loc=<-17.855 0 0>, rot=<0 0 0 1>}
 deform_elbow_in_base_r {parent=lowerarm_r, loc=<-0.211 -5.2591 -0.8275>, rot=<0 0 0 1>}
 deform_elbow_in_r {parent=deform_elbow_in_base_r, loc=<-0.0001 0 0.0002>, rot=<0 0 0 1>}
 deform_elbow_out_base_r {parent=lowerarm_r, loc=<0 4 -0.0001>, rot=<0 0 0 1>}
 deform_elbow_out_r {parent=deform_elbow_out_base_r, loc=<0 0 0>, rot=<0 0 0 1>}
 deform_lowerSleeve_r {parent=lowerarm_r, loc=<0 0 0>, rot=<0 0 0 1>}
 base_lowerSleeveBk_r {parent=deform_lowerSleeve_r, loc=<-14.7487 6.7687 3.2234>, rot=<0 0 0 1>}
 dyn_lowerSleeveBk_r_01 {parent=base_lowerSleeveBk_r, loc=<0 0 0>, rot=<0 0 0 1>}
 base_lowerSleeveFr_r {parent=deform_lowerSleeve_r, loc=<-15.7826 -5.1382 -0.0002>, rot=<0 0 0 1>}
 dyn_lowerSleeveFr_r_01 {parent=base_lowerSleeveFr_r, loc=<0 0 0>, rot=<0 0 0 -1>}
 base_lowerSleeveOut_r {parent=deform_lowerSleeve_r, loc=<-16.3323 -0.3778 -5.007>, rot=<0 0 0 1>}
 dyn_lowerSleeveOut_r_01 {parent=base_lowerSleeveOut_r, loc=<0 0 0>, rot=<0 0 0 1>}
 base_lowerSleeveIn_r {parent=deform_lowerSleeve_r, loc=<-15.4468 -0.3542 7.4647>, rot=<0 0 0 1>}
 dyn_lowerSleeveIn_r_01 {parent=base_lowerSleeveIn_r, loc=<0 0 0>, rot=<0 0 0 1>}
 upperarm_twist_01_r {parent=upperarm_r, loc=<-0.0041 0 0>, rot=<0 0 0 1>}
 deform_upperarm_bk_base_r {parent=upperarm_twist_01_r, loc=<-1.0814 7.2192 1.9456>, rot=<0 0 0 1>}
 deform_upperarm_bk_r {parent=deform_upperarm_bk_base_r, loc=<0 0 0>, rot=<0 0 0 1>}
 deform_upperarm_fr_base_r {parent=upperarm_twist_01_r, loc=<0.3689 -7.0506 2.7652>, rot=<0 0 0 1>}
 deform_upperarm_fr_r {parent=deform_upperarm_fr_base_r, loc=<0 0 0>, rot=<0 0 0 1>}
 deform_upperarm_up_base_r {parent=upperarm_twist_01_r, loc=<3.2214 1.1391 -3.3308>, rot=<0 0 0 1>}
 deform_upperarm_up_r {parent=deform_upperarm_up_base_r, loc=<0 0 0>, rot=<0 0 0 -1>}
 upperarm_twist_02_r {parent=upperarm_r, loc=<-13.7942 0 0>, rot=<0 0 0 1>}
 deform_upperSleeve_r {parent=upperarm_r, loc=<0 0 0>, rot=<0 0 0 1>}
 base_upperSleeveFr_r {parent=deform_upperSleeve_r, loc=<-6.6016 -6.6159 -0.7981>, rot=<0 0 0 1>}
 dyn_upperSleeveFr_r_01 {parent=base_upperSleeveFr_r, loc=<0 0 0>, rot=<0 0 0 -1>}
 base_upperSleeveBk_r {parent=deform_upperSleeve_r, loc=<-7.6197 6.3259 -0.2651>, rot=<0 0 0 1>}
 dyn_upperSleeveBk_r_01 {parent=base_upperSleeveBk_r, loc=<0 0 0>, rot=<0 0 0 1>}
 base_upperSleeveOut_r {parent=deform_upperSleeve_r, loc=<-7.2389 -0.1188 -2.6931>, rot=<0 0 0 1>}
 dyn_upperSleeveOut_r_01 {parent=base_upperSleeveOut_r, loc=<0 0 0>, rot=<0 0 0 1>}
 base_upperSleeveIn_r {parent=deform_upperSleeve_r, loc=<-7.7618 -0.5617 7.3362>, rot=<0 0 0 1>}
 dyn_upperSleeveIn_r_01 {parent=base_upperSleeveIn_r, loc=<0 0 0>, rot=<0 0 0 1>}
 deform_collar {parent=spine_05, loc=<0 0 0>, rot=<0 0 0 1>}
 base_collar_r {parent=deform_collar, loc=<11.9703 2.0061 8.5796>, rot=<-0.2568 -0.6588 -0.5243 0.4745>}
 dyn_collar_r_01 {parent=base_collar_r, loc=<0 0 0>, rot=<0 0 0 1>}
 dyn_collar_r_02 {parent=dyn_collar_r_01, loc=<4 0 0>, rot=<0 0 -0.1305 -0.9914>}
 base_collar_l {parent=deform_collar, loc=<11.9703 2.0061 -8.5797>, rot=<-0.2568 -0.6588 0.5243 -0.4745>}
 dyn_collar_l_01 {parent=base_collar_l, loc=<0 0 0>, rot=<0 0 0 1>}
 dyn_collar_l_02 {parent=dyn_collar_l_01, loc=<4 0 0>, rot=<0 0 -0.1305 -0.9914>}
 base_collar_bk {parent=deform_collar, loc=<14.4245 -6.2643 0>, rot=<0 0 0.8941 -0.4478>}
 dyn_collar_bk_01 {parent=base_collar_bk, loc=<0 0 0>, rot=<0 0 0 1>}
 dyn_collar_bk_02 {parent=dyn_collar_bk_01, loc=<4 0 0>, rot=<0 0 -0.1305 -0.9914>}
 base_collar_front {parent=deform_collar, loc=<6.7364 6.6423 0>, rot=<-0.2379 -0.9713 0 0>}
 dyn_collar_front_01 {parent=base_collar_front, loc=<0 0 0>, rot=<0 0 0 1>}
 dyn_collar_front_02 {parent=dyn_collar_front_01, loc=<4 0 0>, rot=<0 0 -0.1305 -0.9914>}
 base_collar_front_r {parent=deform_collar, loc=<8.7347 5.2164 5.7232>, rot=<-0.2452 -0.9065 -0.1324 0.3172>}
 dyn_collar_front_r_01 {parent=base_collar_front_r, loc=<0 0 0>, rot=<0 0 0 1>}
 dyn_collar_front_r_02 {parent=dyn_collar_front_r_01, loc=<4 0 0>, rot=<0 0 -0.1305 -0.9914>}
 base_collar_front_l {parent=deform_collar, loc=<8.7347 5.2164 -5.7232>, rot=<-0.2452 -0.9065 0.1324 -0.3172>}
 dyn_collar_front_l_01 {parent=base_collar_front_l, loc=<0 0 0>, rot=<0 0 0 1>}
 dyn_collar_front_l_02 {parent=dyn_collar_front_l_01, loc=<4 0 0>, rot=<0 0 -0.1305 -0.9914>}
 base_collar_bk_r {parent=deform_collar, loc=<14.3794 -3.4102 5.9867>, rot=<0.1696 0.2997 0.8341 -0.4309>}
 dyn_collar_bk_r_01 {parent=base_collar_bk_r, loc=<0 0 0>, rot=<0 0 0 1>}
 dyn_collar_bk_r_02 {parent=dyn_collar_bk_r_01, loc=<4 0 0>, rot=<0 0 -0.1305 -0.9914>}
 base_collar_bk_l {parent=deform_collar, loc=<14.3794 -3.4102 -5.9867>, rot=<-0.1696 -0.2997 0.8341 -0.4309>}
 dyn_collar_bk_l_01 {parent=base_collar_bk_l, loc=<0 0 0>, rot=<0 0 0 1>}
 dyn_collar_bk_l_02 {parent=dyn_collar_bk_l_01, loc=<4 0 0>, rot=<0 0 -0.1305 -0.9914>}
 deform_pec_aim_base_l {parent=spine_04, loc=<7.8732 12.6388 -2.4123>, rot=<0.0102 -0.4602 0.3979 -0.7936>}
 deform_pec_aim_l {parent=deform_pec_aim_base_l, loc=<0 0 0>, rot=<0 0 0 1>}
 deform_pec_base_l {parent=deform_pec_aim_l, loc=<10.766 0 0>, rot=<0 0 0 1>}
 deform_pec_l {parent=deform_pec_base_l, loc=<0.055 0 0>, rot=<0 0 0 1>}
 deform_pec_aim_base_r {parent=spine_04, loc=<7.8727 12.6388 2.4123>, rot=<-0.4602 -0.0102 0.7936 0.3979>}
 deform_pec_aim_r {parent=deform_pec_aim_base_r, loc=<0 0 0>, rot=<0 0 0 1>}
 deform_pec_base_r {parent=deform_pec_aim_r, loc=<-10.7662 -0.0006 -0.0007>, rot=<0 0 0 1>}
 deform_pec_r {parent=deform_pec_base_r, loc=<0 0 0>, rot=<0 0 0 1>}
 deform_costume {parent=spine_03, loc=<0 0 0>, rot=<0 0 0 1>}
 base_costume_l {parent=deform_costume, loc=<5.145 -1.3738 -14.6156>, rot=<0 -0.7071 0.7071 0>}
 dyn_costume_l_01 {parent=base_costume_l, loc=<0 0 0>, rot=<0 0 0 1>}
 base_costume_fr {parent=deform_costume, loc=<3.0573 10.4972 0.0044>, rot=<0 -1 0 0>}
 dyn_costume_fr_01 {parent=base_costume_fr, loc=<0 0 0>, rot=<0 0 0 1>}
 base_chest_softball {parent=dyn_costume_fr_01, loc=<0 0.5 0>, rot=<0 0 0 1>}
 dyn_chest_softball {parent=base_chest_softball, loc=<0 0 0>, rot=<0 0 0 1>}
 base_costume_bk_l {parent=deform_costume, loc=<7.0308 -11.9493 -6>, rot=<0 0 1 0>}
 dyn_costume_bk_l_01 {parent=base_costume_bk_l, loc=<0 0 0>, rot=<0 0 0 1>}
 base_costume_bk_r {parent=deform_costume, loc=<7.0308 -11.9493 6>, rot=<0 0 1 0>}
 dyn_costume_bk_r_01 {parent=base_costume_bk_r, loc=<0 0 0>, rot=<0 0 0 1>}
 base_costume_r {parent=deform_costume, loc=<5.145 -1.3738 14.6154>, rot=<0 0.7071 0.7071 0>}
 dyn_costume_r_01 {parent=base_costume_r, loc=<0 0 0>, rot=<0 0 0 1>}
 thigh_l {parent=pelvis, loc=<-2.5693 -0.0109 -7.7501>, rot=<0.0012 0.0572 0.0155 -0.9982>}
 calf_l {parent=thigh_l, loc=<-39.0484 0 0>, rot=<0.0006 -0.0161 -0.0391 -0.9991>}
 foot_l {parent=calf_l, loc=<-42.561 0.1077 0.4292>, rot=<0.0418 -0.0455 0.0343 -0.9975>}
 ball_l {parent=foot_l, loc=<-6.1599 13.8264 0.6806>, rot=<0.0001 -0.0001 -0.7063 -0.7079>}
 deform_foot_bk_l {parent=foot_l, loc=<-1.3424 -5.1158 0.6012>, rot=<0.7027 -0.71 -0.0302 -0.0339>}
 calf_twist_02_l {parent=calf_l, loc=<-18.8064 -0.0967 0.0397>, rot=<0.0237 0.0095 -0.0035 -0.9997>}
 calf_twist_01_l {parent=calf_l, loc=<-35.813 -0.0967 0.0397>, rot=<0.0237 0.0095 -0.0035 -0.9997>}
 deform_knee_in_base_l {parent=calf_l, loc=<1.298 -5.2957 -0.0496>, rot=<0 0 0 -1>}
 deform_knee_in_l {parent=deform_knee_in_base_l, loc=<0 0 0>, rot=<0 0 0 1>}
 deform_knee_out_base_l {parent=calf_l, loc=<0 4 0>, rot=<0 0 0 -1>}
 deform_knee_out_l {parent=deform_knee_out_base_l, loc=<0 0 0>, rot=<0 0 0 1>}
 deform_lowerPant_l {parent=calf_l, loc=<0 0 0>, rot=<0 0 0 1>}
 base_lowerPantBk_l {parent=deform_lowerPant_l, loc=<-41.4246 -6.7687 -0.8439>, rot=<0 0 0 1>}
 dyn_lowerPantBk_l_01 {parent=base_lowerPantBk_l, loc=<0 0 0>, rot=<0 0 0 1>}
 base_lowerPantFr_l {parent=deform_lowerPant_l, loc=<-37.9948 4.1727 0.0002>, rot=<0 0 0 1>}
 dyn_lowerPantFr_l_01 {parent=base_lowerPantFr_l, loc=<0 0 0>, rot=<0 0 0 -1>}
 base_lowerPantIn_l {parent=deform_lowerPant_l, loc=<-35.4778 0.3778 6.8103>, rot=<0 0 0 1>}
 dyn_lowerPantIn_l_01 {parent=base_lowerPantIn_l, loc=<0 0 0>, rot=<0 0 0 1>}
 base_lowerPantOut_l {parent=deform_lowerPant_l, loc=<-35.4326 0.3542 -4.3182>, rot=<0 0 0 1>}
 dyn_lowerPantOut_l_01 {parent=base_lowerPantOut_l, loc=<0 0 0>, rot=<0 0 0 -1>}
 thigh_twist_01_l {parent=thigh_l, loc=<-21.6303 0 0>, rot=<0.0013 0 0 -1>}
 deform_groin_base_l {parent=thigh_l, loc=<-2.1908 8.1142 -1.7442>, rot=<0 0 0 -1>}
 deform_groin_l {parent=deform_groin_base_l, loc=<0 0 0>, rot=<0 0 0 1>}
 thigh_r {parent=pelvis, loc=<-2.5691 -0.0109 7.7501>, rot=<0.0572 -0.0012 0.9982 0.0155>}
 calf_r {parent=thigh_r, loc=<39.0487 0 0>, rot=<0.0006 -0.0161 -0.0391 -0.9991>}
 foot_r {parent=calf_r, loc=<42.5612 -0.1077 -0.4292>, rot=<0.0418 -0.0455 0.0343 -0.9975>}
 ball_r {parent=foot_r, loc=<6.1598 -13.8264 -0.6806>, rot=<0.0001 -0.0001 -0.7063 -0.7079>}
 deform_foot_bk_r {parent=foot_r, loc=<1.3663 5.1156 -0.6>, rot=<0.71 0.7027 0.0339 -0.0302>}
 calf_twist_02_r {parent=calf_r, loc=<18.8064 0.0967 -0.0397>, rot=<0.0237 0.0095 -0.0035 -0.9997>}
 calf_twist_01_r {parent=calf_r, loc=<35.8132 0.0967 -0.0397>, rot=<0.0237 0.0095 -0.0035 -0.9997>}
 deform_knee_in_base_r {parent=calf_r, loc=<-1.2981 5.2958 0.0496>, rot=<0 0 0 1>}
 deform_knee_in_r {parent=deform_knee_in_base_r, loc=<0 0 0>, rot=<0 0 0 1>}
 deform_knee_out_base_r {parent=calf_r, loc=<-0.0002 -4 0>, rot=<0 0 0 -1>}
 deform_knee_out_r {parent=deform_knee_out_base_r, loc=<0 0 0>, rot=<0 0 0 1>}
 deform_lowerPant_r {parent=calf_r, loc=<0 0 0>, rot=<0 0 0 1>}
 base_lowerPantBk_r {parent=deform_lowerPant_r, loc=<41.4246 6.7687 0.8439>, rot=<0 0 0 1>}
 dyn_lowerPantBk_r_01 {parent=base_lowerPantBk_r, loc=<0 0 0>, rot=<0 0 0 1>}
 base_lowerPantFr_r {parent=deform_lowerPant_r, loc=<37.9948 -4.1727 -0.0002>, rot=<0 0 0 1>}
 dyn_lowerPantFr_r_01 {parent=base_lowerPantFr_r, loc=<0 0 0>, rot=<0 0 0 -1>}
 base_lowerPantIn_r {parent=deform_lowerPant_r, loc=<35.4778 -0.3778 -6.8103>, rot=<0 0 0 1>}
 dyn_lowerPantIn_r_01 {parent=base_lowerPantIn_r, loc=<0 0 0>, rot=<0 0 0 1>}
 base_lowerPantOut_r {parent=deform_lowerPant_r, loc=<35.4326 -0.3542 4.3182>, rot=<0 0 0 1>}
 dyn_lowerPantOut_r_01 {parent=base_lowerPantOut_r, loc=<0 0 0>, rot=<0 0 0 1>}
 thigh_twist_01_r {parent=thigh_r, loc=<21.6303 0 0>, rot=<0.0013 0 0 -1>}
 deform_groin_base_r {parent=thigh_r, loc=<2.1908 -8.1142 1.7442>, rot=<0 0 0 -1>}
 deform_groin_r {parent=deform_groin_base_r, loc=<0 0 0>, rot=<0 0 0 1>}
 deform_glute_base_l {parent=pelvis, loc=<-8.8064 -8.8222 -8.4287>, rot=<0.0275 0.0501 -0.4472 -0.8926>}
 deform_glute_l {parent=deform_glute_base_l, loc=<0 0 0>, rot=<0 0 0 1>}
 deform_glute_base_r {parent=pelvis, loc=<-8.8062 -8.8222 8.4287>, rot=<0.0501 -0.0275 0.8926 -0.4472>}
 deform_glute_r {parent=deform_glute_base_r, loc=<0 0 0>, rot=<0 0 0 1>}
 base_pelvis_softball {parent=pelvis, loc=<-2.2875 9.0359 0.0044>, rot=<-0.0179 -0.9998 0 0>}
 dyn_pelvis_softball {parent=base_pelvis_softball, loc=<0 0 0>, rot=<0 0 0 1>}
]
variables [
 Aim {type=bool, default=False}
 ArmsHidden {type=bool, default=False}
 DBNO {type=bool, default=False}
 Emote {type=bool, default=False}
 LegsHidden {type=bool, default=False}
 Melee {type=bool, default=False}
 NotRBAN {type=bool, default=False}
 Pistol {type=bool, default=False}
]
